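; Thrown C4 projectile actor for Figment Athena: sticks to surfaces and vehicles or bounces, detonates when triggered or damaged, and can be picked up.
; The server handles attachment, re-attaching or resuming physics when bases die.
; Detonation sends gameplay events to abilities and destroys the actor, with damage crediting the attacker.
; Tunable flags govern owner-death behavior; team-colored beacon, water handling, landing sounds and explosion effects provide feedback.
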(event Custom|VehicleActorDestroyed
  (params (DestroyedActor :type Actor%20Object%20Reference))
  ;[0.3021 0.09068 0.2563 0.5] FORT-983093; re-enabling collision once vehicle is destoyed
  (Collision|SetCollisionEnabled (Variables|Default|GetRoot_Sphere_Collision) "QueryAndPhysics")
  (Collision|SetCollisionResponsetoChannel (Variables|Default|GetSM_C4_Explosive) "ECC_Pawn" "ECR_Overlap"))

(event Custom|ExplodeOnOwnerDeath
  (params (DamagedActor :type Actor%20Object%20Reference) (Damage :type Float%20%28single-precision%29) (InstigatedBy :type Controller%20Object%20Reference) (DamageCauser :type Actor%20Object%20Reference) (HitLocation :type Vector) (FHitComponent :type Primitive%20Component%20Object%20Reference) (BoneName :type Name) (Momentum :type Vector))
  ;[1 1 1] when your owner dies, see if you are supposed to destroy yourself
  (Utilities|FlowControl|Sequence
    (:then_0
      ;[1 1 1] FORT-520290 Clear the ignore actors when player owner dies, otherwise they remain as a dangling reference.
      (bind _original_player_owner (Variables|Default|GetOriginal_Player_Owner)
        (:then
          (for _array_element (Actor|GetAttachedActors _original_player_owner true false)
            (Collision|IgnoreActorwhenMoving (Variables|Default|GetRoot_Sphere_Collision) _array_element false)
            (Collision|IgnoreActorwhenMoving (Variables|Default|GetSM_C4_Explosive) _array_element false)))
        (:else)))
    (:then_1
      (if (ScalableFloat|GetValueasBool (Variables|Default|Getdestroydeath))
        (Actor|DestroyActor)))))

(event EventBeginPlay
  (bind _returnvalue (Game|GetInstigator))
  (bind _original_player_owner (Variables|Default|GetOriginal_Player_Owner))
  (bind _returnvalue_2 (Game|GetPlayerController 0))
  (bind _returnvalue_3 (Pawn|GetControlledPawn _returnvalue_2))
  (bind _returnvalue_4 (Team|OnSameTeam _returnvalue_3 _original_player_owner))
  (bind _returnvalue_5 (select _returnvalue_4 "(R=0.000000,G=0.200000,B=0.800000,A=1.000000)" "(R=1.000000,G=0.000000,B=0.000000,A=1.000000)"))
  (Networking|SwitchHasAuthority
    (:Authority
      (AI|Perception|RegisterwithPerceptionSystem (Variables|Default|GetAthenaAIPerceptionStimuliSource)))
    (:Remote
      _))
  ;[1 1 1] Save the original owner to handle any needed events with him
  (Variables|Interaction|SetbAllowInteract false)
  (bind _asfort_player_pawn_athena (Utilities|Casting|CastToFortPlayerPawnAthena (Game|GetInstigator)))
  (bind _output_get (Variables|Default|SetOriginal_Player_Owner _asfort_player_pawn_athena))
  ;[0 0 0] Ignore actors attached to instigator pawn (i.e. bush)
  (bind _output_get_2 (Variables|Default|SetActorsToIgnore (Actor|GetAttachedActors _output_get true false)))
  (for _array_element _output_get_2
    (Collision|IgnoreActorwhenMoving (Variables|Default|GetRoot_Sphere_Collision) _array_element true)
    (Collision|IgnoreActorwhenMoving (Variables|Default|GetSM_C4_Explosive) _array_element true))
  (EventDispatchers|BindEventtoOnDied _asfort_player_pawn_athena (delegate Custom|ExplodeOnOwnerDeath))
  (Utilities|FlowControl|Sequence
    (:then_0
      ;[1 1 1] Set Beacon Color based on team
      (Networking|CanExecuteCosmeticEvents
        (:True
          (Effects|Components|ParticleSystem|SetColorParameter (Variables|Default|GetNS_C4_Flare) "TeamColor" _returnvalue_5)
          (Rendering|Material|SetVectorParameterValueonMaterials (Variables|Default|GetSM_C4_Explosive) "Emissive Multiplier" (Math|Conversions|ToVector(LinearColor) _returnvalue_5)))))
    (:then_1
      ;[1 1 1] On Load
      (Variables|/Script/Engine.ProjectileMovementComponent|SetBounciness 0.2 (Variables|Default|GetProjectileMovement1))
      (if (Utilities|IsValid _returnvalue :!class /Script/Engine.KismetSystemLibrary)
        (CallFunction|SetTeam (Team|GetActorTeam _returnvalue))))
    (:then_2
      ;[1 1 1] FORT-457568, FORT-459668
      ; Traces X units ahead of the player to prevent throwing through walls.
      (Networking|SwitchHasAuthority
        (:Authority
          (bind (_outhit _returnvalue_6) (Collision|LineTraceByChannel (Transformation|GetActorLocation) (+ (Transformation|GetActorLocation) (* (Math|Vector|Normalize (Variables|/Script/Engine.MovementComponent|GetVelocity (Variables|Default|GetProjectileMovement1))) (Variables|Default|GetTraceAheadDistance))) "TraceTypeQuery6" true (Variables|Default|GetActorsToIgnore)))
          (if _returnvalue_6
            (CallFunction|OnThrowHitImmediately _outhit)))))))

(event Custom|SetExplosion
  (params (Explosion%20Level :type Integer) (Activation_Radius :type Float%20%28double-precision%29) (PlayerPawn :type Fort%20Player%20Pawn%20Object%20Reference))
  (Variables|Default|SetActivation_Radius Activation_Radius)
  (Variables|Default|SetPlayer_Owner PlayerPawn)
  (Variables|Default|SetExplosionLevel Explosion%20Level)
  ;[0.1261 0.1093 1] Scaling the explosion
  (Variables|Default|SetEventExplosion (Variables|Default|GetExplosionRangeDefaultTag)))

(event EventOnDeathServer
  (params (Damage :type Float%20%28single-precision%29) (DamageTags :type Gameplay%20Tag%20Container%20Structure%20%28by%20ref%29) (Momentum :type Vector :default "0, 0, 0") (HitInfo :type Hit%20Result%20Structure%20%28by%20ref%29) (InstigatedBy :type Controller%20Object%20Reference) (DamageCauser :type Actor%20Object%20Reference) (EffectContext :type Gameplay%20Effect%20Context%20Handle%20Structure))
  ;[1 1 1] When this is destroyed, check to see if its supposed to explode
  (if (ScalableFloat|GetValueasBool (Variables|Default|GetExplodeDeath))
    (CallFunction|DamageExplode 0)))

(event Custom|BaseDied
  (params (DamagedActor :type Actor%20Object%20Reference) (Damage :type Float%20%28single-precision%29) (InstigatedBy :type Controller%20Object%20Reference) (DamageCauser :type Actor%20Object%20Reference) (HitLocation :type Vector) (FHitComponent :type Primitive%20Component%20Object%20Reference) (BoneName :type Name) (Momentum :type Vector))
  ;[1 1 1] Base destroyed. Start simulating again.
  (bind (_outhit _returnvalue) (Collision|SphereTraceForObjects (Transformation|GetActorLocation) (Transformation|GetActorLocation) 16.0 (Utilities|Array|MakeArray "ObjectTypeQuery1")))
  (if (and (Utilities|Select :Index (Collision|GetCollisionResponsetoChannel (.HitComponent (Collision|BreakHitResult _outhit)) "ECC_GameTraceChannel8") :ECR_Ignore false :ECR_Overlap false :ECR_Block true) _returnvalue)
    (CallFunction|HandleWasAttachedtoWall (.HitActor (Collision|BreakHitResult _outhit)))
    (bind _asbuilding_actor (Utilities|Casting|CastToBuildingActor (.HitActor (Collision|BreakHitResult _outhit)))
      (:then
        (EventDispatchers|BindEventtoOnDied _asbuilding_actor (EventDispatchers|CreateEvent :!event Custom|BaseDied)))
      (:CastFailed
        (Game|BindEventtoOnDestroyed _asbuilding_actor (EventDispatchers|CreateEvent :!event Custom|BaseDestroyed))))
    (else
      (Variables|Default|SetbResumeSimulation true)
      (Networking|ForceNetUpdate))))

(event Custom|SetProjectile
  (params (Init_Speed :type Float%20%28double-precision%29) (Gravity :type Float%20%28double-precision%29) (Activation_Delay :type Float%20%28double-precision%29))
  (bind _projectilemovement1 (Variables|Default|GetProjectileMovement1))
  ;[1 0.03636 0] Sets Prj movement from GA  (Not getting values!!!!)
  (Variables|/Script/Engine.ProjectileMovementComponent|SetInitialSpeed Init_Speed _projectilemovement1)
  (Variables|/Script/Engine.ProjectileMovementComponent|SetProjectileGravityScale Gravity _projectilemovement1)
  (Variables|Default|SetActivation_Delay Activation_Delay)
  (Development|PrintString "fires?"))

(event Custom|BaseDestroyed
  (params (DestroyedActor :type Actor%20Object%20Reference))
  ;[1 1 1] Base destroyed. Start simulating again.
  (bind (_outhit _returnvalue) (Collision|SphereTraceForObjects (Transformation|GetActorLocation) (Transformation|GetActorLocation) 16.0 (Utilities|Array|MakeArray "ObjectTypeQuery1")))
  (if (and (Utilities|Select :Index (Collision|GetCollisionResponsetoChannel (.HitComponent (Collision|BreakHitResult _outhit)) "ECC_GameTraceChannel8") :ECR_Ignore false :ECR_Overlap false :ECR_Block true) _returnvalue)
    (CallFunction|HandleWasAttachedtoWall (.HitActor (Collision|BreakHitResult _outhit)))
    (bind _asbuilding_actor (Utilities|Casting|CastToBuildingActor (.HitActor (Collision|BreakHitResult _outhit)))
      (:then
        (EventDispatchers|BindEventtoOnDied _asbuilding_actor (EventDispatchers|CreateEvent :!event Custom|BaseDied)))
      (:CastFailed
        (Game|BindEventtoOnDestroyed _asbuilding_actor (EventDispatchers|CreateEvent :!event Custom|BaseDestroyed))))
    (else
      (Variables|Default|SetbResumeSimulation true)
      (Networking|ForceNetUpdate))))

(event Custom|OnThrow_HitImmediately
  (params (Hit :type Hit%20Result%20Structure))
  (bind _hitresult (Variables|Default|GetHitResult))
  (bind (_bblockinghit _binitialoverlap _time _distance _location _impactpoint _normal _impactnormal _physmat _hitactor) (Collision|BreakHitResult _hitresult))
  (bind (_bblockinghit_2 _binitialoverlap_2 _time_2 _distance_2 _location_2 _impactpoint_2 _normal_2 _impactnormal_2 _physmat_2 _hitactor_2 _hitcomponent) (Collision|BreakHitResult (Variables|Default|GetHitResult)))
  ;[0 1 0.3181] When this actor hits something, if its not a pawn, attach to it!
  (Variables|Default|SetHitResult Hit)
  (if (not (Variables|Default|GetInWater))
    (Development|PrintString "C4 HIT")
    (Networking|SwitchHasAuthority
      (:Authority
        (if (not (Components|ComponentHasTag _hitcomponent "NoStickC4"))
          (bind _bdidactorhavetags (Class|AthenaQuestFunctionLibrary|AthenaCheckActorHasTags _hitactor (Variables|Default|GetNo%20Stick%20Tags)))
          (if (not _bdidactorhavetags)
            (CallFunction|TryAttachtoActor _hitactor)))
        ;[1 0.9028 0.2728] Bounce off actor if not sticking
        (Variables|/Script/Engine.MovementComponent|SetVelocity (Math|Vector|vector+vector (Math|Vector|ClampVectorSize (+ (/ (Math|Vector|MirrorVectorbyNormal (Variables|/Script/Engine.MovementComponent|GetVelocity (Variables|Default|GetProjectileMovement1)) "0, 0,1.000000") 2.0) (Math|Vector|CrossProduct (/ (Math|Vector|MirrorVectorbyNormal (Variables|/Script/Engine.MovementComponent|GetVelocity (Variables|Default|GetProjectileMovement1)) "0, 0,1.000000") 2.0) (Math|Vector|MakeVector (Math|Random|RandomFloatinRange -0.5 0.5) (Math|Random|RandomFloatinRange -0.5 0.5)))) 4e+02 7.5e+02) (0.0 0.0 1e+02)) (Variables|Default|GetProjectileMovement1))
        (Variables|Default|SetAllowExplode? true)))))

(event Collision|EventHit
  (params (MyComp :type Primitive%20Component%20Object%20Reference) (Other :type Actor%20Object%20Reference) (OtherComp :type Primitive%20Component%20Object%20Reference) (bSelfMoved :type Boolean) (HitLocation :type Vector :default "0, 0, 0") (HitNormal :type Vector :default "0, 0, 0") (NormalImpulse :type Vector :default "0, 0, 0") (Hit :type Hit%20Result%20Structure%20%28by%20ref%29))
  (bind _hitresult (Variables|Default|GetHitResult))
  (bind (_bblockinghit _binitialoverlap _time _distance _location _impactpoint _normal _impactnormal _physmat _hitactor) (Collision|BreakHitResult _hitresult))
  (bind (_bblockinghit_2 _binitialoverlap_2 _time_2 _distance_2 _location_2 _impactpoint_2 _normal_2 _impactnormal_2 _physmat_2 _hitactor_2 _hitcomponent) (Collision|BreakHitResult (Variables|Default|GetHitResult)))
  ;[0 1 0.3181] When this actor hits something, if its not a pawn, attach to it!
  (Variables|Default|SetHitResult Hit)
  (if (not (Variables|Default|GetInWater))
    (Development|PrintString "C4 HIT")
    (Networking|SwitchHasAuthority
      (:Authority
        (if (not (Components|ComponentHasTag _hitcomponent "NoStickC4"))
          (bind _bdidactorhavetags (Class|AthenaQuestFunctionLibrary|AthenaCheckActorHasTags _hitactor (Variables|Default|GetNo%20Stick%20Tags)))
          (if (not _bdidactorhavetags)
            (CallFunction|TryAttachtoActor _hitactor)))
        ;[1 0.9028 0.2728] Bounce off actor if not sticking
        (Variables|/Script/Engine.MovementComponent|SetVelocity (Math|Vector|vector+vector (Math|Vector|ClampVectorSize (+ (/ (Math|Vector|MirrorVectorbyNormal (Variables|/Script/Engine.MovementComponent|GetVelocity (Variables|Default|GetProjectileMovement1)) "0, 0,1.000000") 2.0) (Math|Vector|CrossProduct (/ (Math|Vector|MirrorVectorbyNormal (Variables|/Script/Engine.MovementComponent|GetVelocity (Variables|Default|GetProjectileMovement1)) "0, 0,1.000000") 2.0) (Math|Vector|MakeVector (Math|Random|RandomFloatinRange -0.5 0.5) (Math|Random|RandomFloatinRange -0.5 0.5)))) 4e+02 7.5e+02) (0.0 0.0 1e+02)) (Variables|Default|GetProjectileMovement1))
        (Variables|Default|SetAllowExplode? true)))))

(event OnProjectileAttached(FortAttachableProjectile)
  (params (ActorAttachedTo :type Actor%20Object%20Reference) (WorldLocation :type Vector%20%28by%20ref%29 :default "0, 0, 0") (AttachmentNormal :type Vector%20%28by%20ref%29 :default "0, 0, 0"))
  (bind (_asfort_pawn _bsuccess) (Utilities|Casting|CastToFortPawn ActorAttachedTo))
  ;[0.07785 1 0.7891] On attached
  (CallFunction|HandleHitVehicleorCube ActorAttachedTo)
  (Variables|Default|SetbSkipRotationAdjustment _bsuccess) ; Custom rotation already applied for pawns
  (bind _output_get (Variables|Default|SetImpact_Loc WorldLocation))
  (Variables|Default|SetHitNormal AttachmentNormal)
  (Transformation|SetActorLocation _output_get false true) ; Ensure actor is at the attachment location. \nServer will attach at center of pawn, and that will \nsometimes replicate to the client
  (Utilities|FlowControl|Sequence
    (:then_0
      (Networking|CanExecuteCosmeticEvents
        (:True
          ;;[0.7396 0.3397 1] FX goo
          (bind _returnvalue (Niagara|SpawnSystemAttached "/Game/Athena/Items/Consumables/C4/NS_C4_Goo.NS_C4_Goo" (Variables|Default|GetSM_C4_Explosive))) ; FX impact goo
          (Variables|Default|SetFXImpact _returnvalue)
          (Niagara|SpawnSystematLocation "/Game/Athena/Items/Consumables/C4/NS_C4_TriggerFlare.NS_C4_TriggerFlare" (+ (Transformation|GetWorldLocation (Variables|Default|GetSM_C4_Explosive)) (* (Transformation|GetUpVector (Variables|Default|GetSM_C4_Explosive)) 5.0))) ; Flare Flash
        )))
    (:then_1
      (CallFunction|OnC4Placed))))

(event Custom|On%20C4%20Placed
  ;[0.1534 1 0.7367] On C4 Placed
  ;;[0.9194 1 0.3694] Delay the ability to explode until it stops moving
  (Variables|Default|SetAllowExplode? true)
  (Networking|SwitchHasAuthority
    (:Authority
      (if (Variables|Default|GetbTriggeredExplosion)
        (CallFunction|TriggeredExplosion 0.05)
        (else
          (Variables|Default|SetLandedTime (GameState|GetServerWorldTimeSeconds (Game|GetGameState)))
          (AI|Perception|UnregisterfromPerceptionSystem (Variables|Default|GetAthenaAIPerceptionStimuliSource)))))
    (:Remote
      (Audio|Components|Audio|FadeOut (Variables|Default|GetIn%20Air%20Sound) 0.25 0.0 "Linear" :!class /Script/Engine.AudioComponent)))
  ;; MinePlaced
  (Variables|Default|SetRotationInAir (Transformation|GetActorRotation self))
  (if (not (or (Variables|Default|GetInWater) (Variables|Default|GetbSkipRotationAdjustment)))
    (Transformation|SetActorRotation (Math|Rotator|MakeRotfromZ (Variables|Default|GetHitNormal)))
    (Transformation|AddRelativeRotation (Variables|Default|GetSM_C4_Explosive) (Math|Rotator|MakeRotator 0.0 0.0 (- (.yaw (Variables|Default|GetRotationInAir)) 9e+01)) false false))
  (Variables|Interaction|SetbAllowInteract true)
  (Networking|CanExecuteCosmeticEvents
    (:True
      (Effects|SpawnEmitteratLocation "/Game/Athena/Items/Consumables/C4/P_Athena_C4_Land_Dust.P_Athena_C4_Land_Dust" (Variables|Default|GetImpact_Loc) (Math|Conversions|RotationFromXVector (Variables|Default|GetHitNormal)))
      (Components|Activation|SetActive (Variables|Default|GetNS_C4_Flare) true false))))

(event Custom|DamageExplode
  (params (DamageOwner :type Fort%20Player%20Pawn%20Athena%20Object%20Reference))
  (bind _returnvalue (Game|GetInstigator))
  (bind _original_player_owner (Variables|Default|GetOriginal_Player_Owner))
  (bind _player_owner (Variables|Default|GetPlayer_Owner))
  ;[1 0.75 0] Damage caused this to explode
  (bind _output_get (Variables|Default|SetPlayer_Owner DamageOwner))
  (Utilities|IsValid _output_get
    (:Is%20Valid
      (bind _asfort_player_controller (Utilities|Casting|CastToFortPlayerController (Pawn|GetController _output_get))
        (:then
          (AbilitySystem|ForceChangeOwnerInstigator _asfort_player_controller))
        (:CastFailed
          _))
      (Utilities|FlowControl|DoOnce false
        (:Completed
          ;[0.8855 0.04538 1]
          (Variables|Default|SetIsDead? true)
          (Networking|SwitchHasAuthority
            (:Authority
              (Utilities|IsValid _player_owner
                (:Is%20Valid
                  (Ability|SendGameplayEventtoActor self (Variables|Default|GetEventExplosion) (Utilities|Struct|MakeGameplayEventData (Variables|Default|GetEventExplosion) _player_owner _player_owner self))
                  (Utilities|IsValid _returnvalue
                    (:Is%20Valid
                      (Ability|SendGameplayEventtoActor _returnvalue (Variables|Default|GetEventExplosion) (Utilities|Struct|MakeGameplayEventData (Variables|Default|GetEventExplosion) _returnvalue _returnvalue self)))
                    (:Is%20Not%20Valid
                      _))
                  ;[0.9946 0 1] Add this as a fix so that if an enemy shoots your C4 while you have 0 charges left it is correctly removed from your inventory.
                  (if (Utilities|NotEqual(Object) (Variables|Default|GetOriginal_Player_Owner) (Variables|Default|GetPlayer_Owner))
                    (Utilities|IsValid _original_player_owner
                      (:Is%20Valid
                        (Ability|SendGameplayEventtoActor _original_player_owner (Variables|Default|GetEventExplosion) (Utilities|Struct|MakeGameplayEventData (Variables|Default|GetEventExplosion) _original_player_owner _original_player_owner self)))
                      (:Is%20Not%20Valid
                        _))
                    (Networking|ForceNetUpdate)
                    (Variables|Default|SetbExploded true)
                    (Actor|DestroyActor)
                    (else
                      _)))))) ; If touched or damage\\ Only on server will it trigger explosion
        )))
    (:Is%20Not%20Valid
      _)))

(event OnEnterWaterEvent(FortWaterInteraction)
  (params (WaterBody :type Fort%20Water%20Body%20Actor%20Object%20Reference) (WaterInteractionComponent :type Fort%20Water%20Interaction%20Component%20Object%20Reference) (bIsFirstBody :type Boolean))
  (bind _ns_object_floatinginwater (Variables|Default|GetNS_Object_FloatingInWater))
  ;[0 0.1658 1] Water Handling
  ;;[0.497 0.295 1] On Water enter play VFX and SFX also set the in water bool for tracking. On the server feed back into the normal flow, Bypassing the hit section of the Hit Event. On Client call the mine placed macro.
  (Development|PrintString "Entered Water")
  (bind _outwatersurfaceinfo (WaterBody|GetWaterSurfaceInfoStruct WaterBody (Transformation|GetActorLocation) true false false))
  (Utilities|FlowControl|Sequence
    (:then_0
      (Components|Activation|Activate _ns_object_floatinginwater false)
      (Effects|Components|ParticleSystem|SetFloatParameter _ns_object_floatinginwater "User.WaterDepth" (.WaterDepth (Utilities|Struct|BreakWaterSurfaceInfo _outwatersurfaceinfo)))
      (Effects|Components|ParticleSystem|SetVectorParameter _ns_object_floatinginwater "User.WaterPlaneLocation" (Utilities|Struct|BreakWaterSurfaceInfo _outwatersurfaceinfo))
      (Effects|Components|ParticleSystem|SetVectorParameter _ns_object_floatinginwater "User.WaterPlaneNormal" (.WaterPlaneNormal (Utilities|Struct|BreakWaterSurfaceInfo _outwatersurfaceinfo)))
      (Effects|Components|ParticleSystem|SetFloatParameter _ns_object_floatinginwater "User.InheritVelocityScale" 0.3))
    (:then_1
      (bind _returnvalue (Effects|Components|ParticleSystem|SpawnFXAtLocation (Variables|Default|GetNS_HitWater) (Utilities|Struct|BreakWaterSurfaceInfo _outwatersurfaceinfo) (Math|Rotator|MakeRotfromX (.WaterPlaneNormal (Utilities|Struct|BreakWaterSurfaceInfo _outwatersurfaceinfo)))))
      (Effects|Components|ParticleSystem|SetFloatParameter _returnvalue "WaterDepth" (.WaterDepth (Utilities|Struct|BreakWaterSurfaceInfo _outwatersurfaceinfo)))
      (Effects|Components|ParticleSystem|SetFloatParameter _returnvalue "WaterBodyIndex" (Math|Conversions|ToFloat(Integer) (.WaterBodyIdx (Utilities|Struct|BreakWaterSurfaceInfo _outwatersurfaceinfo))))
      (Variables|Default|SetInWater true)
      (CallFunction|OnC4Placed))))

(event Custom|TriggeredExplosion
  (params (DelayForChainExplode :type Float%20%28double-precision%29))
  (bind _returnvalue (Game|GetInstigator))
  (bind _original_player_owner (Variables|Default|GetOriginal_Player_Owner))
  (bind _player_owner (Variables|Default|GetPlayer_Owner))
  (Variables|Default|SetbTriggeredExplosion true)
  (if (Variables|Default|GetAllowExplode?)
    (Utilities|FlowControl|DoOnce false
      (:Completed
        (Utilities|FlowControl|Delay DelayForChainExplode)
        (Audio|PlaySoundatLocation (Variables|Default|GetPreExplosion_Sound) (Transformation|GetWorldLocation (Variables|Default|GetSM_C4_Explosive)))
        (Utilities|FlowControl|DoOnce false
          (:Completed
            ;[0.8855 0.04538 1]
            (Variables|Default|SetIsDead? true)
            (Networking|SwitchHasAuthority
              (:Authority
                (Utilities|IsValid _player_owner
                  (:Is%20Valid
                    (Ability|SendGameplayEventtoActor self (Variables|Default|GetEventExplosion) (Utilities|Struct|MakeGameplayEventData (Variables|Default|GetEventExplosion) _player_owner _player_owner self))
                    (Utilities|IsValid _returnvalue
                      (:Is%20Valid
                        (Ability|SendGameplayEventtoActor _returnvalue (Variables|Default|GetEventExplosion) (Utilities|Struct|MakeGameplayEventData (Variables|Default|GetEventExplosion) _returnvalue _returnvalue self)))
                      (:Is%20Not%20Valid
                        _))
                    ;[0.9946 0 1] Add this as a fix so that if an enemy shoots your C4 while you have 0 charges left it is correctly removed from your inventory.
                    (if (Utilities|NotEqual(Object) (Variables|Default|GetOriginal_Player_Owner) (Variables|Default|GetPlayer_Owner))
                      (Utilities|IsValid _original_player_owner
                        (:Is%20Valid
                          (Ability|SendGameplayEventtoActor _original_player_owner (Variables|Default|GetEventExplosion) (Utilities|Struct|MakeGameplayEventData (Variables|Default|GetEventExplosion) _original_player_owner _original_player_owner self)))
                        (:Is%20Not%20Valid
                          _))
                      (Networking|ForceNetUpdate)
                      (Variables|Default|SetbExploded true)
                      (Actor|DestroyActor)
                      (else
                        _)))))) ; If touched or damage\\ Only on server will it trigger explosion
          ))))))

(event EventOnDamageServer
  (params (Damage :type Float%20%28single-precision%29) (DamageTags :type Gameplay%20Tag%20Container%20Structure%20%28by%20ref%29) (Momentum :type Vector :default "0, 0, 0") (HitInfo :type Hit%20Result%20Structure%20%28by%20ref%29) (InstigatedBy :type Controller%20Object%20Reference) (DamageCauser :type Actor%20Object%20Reference) (EffectContext :type Gameplay%20Effect%20Context%20Handle%20Structure))
  ;[1 1 1] On Damage switch the owner to the damage dealer and explode it
  (Networking|SwitchHasAuthority
    (:Authority
      (bind _asfort_player_pawn_athena (Utilities|Casting|CastToFortPlayerPawnAthena (Pawn|GetControlledPawn InstigatedBy)))
      (CallFunction|DamageExplode _asfort_player_pawn_athena))))

(event OnExitWaterEvent(FortWaterInteraction)
  (params (WaterBody :type Fort%20Water%20Body%20Actor%20Object%20Reference) (WaterInteractionComponent :type Fort%20Water%20Interaction%20Component%20Object%20Reference) (bIsLastBody :type Boolean))
  ;[0 0.1658 1] Water Handling
  ;;[1 0.533 0.1939] On Exit Water
  (Components|Activation|Deactivate (Variables|Default|GetNS_Object_FloatingInWater)))

(event EventOnDeathPlayEffects
  (params (Damage :type Float%20%28single-precision%29) (DamageTags :type Gameplay%20Tag%20Container%20Structure%20%28by%20ref%29) (Momentum :type Vector :default "0, 0, 0") (HitInfo :type Hit%20Result%20Structure%20%28by%20ref%29) (InstigatedBy :type Fort%20Pawn%20Object%20Reference) (DamageCauser :type Actor%20Object%20Reference) (EffectContext :type Gameplay%20Effect%20Context%20Handle%20Structure))
  (bind _sm_c4_explosive (Variables|Default|GetSM_C4_Explosive))
  (bind _returnvalue (Transformation|GetWorldLocation _sm_c4_explosive))
  (bind _explosion_sound (Variables|Default|GetExplosion_Sound))
  ;[1 0.4852 0] Trigger effects on death
  (Utilities|FlowControl|DoOnce false
    (:Completed
      (Rendering|SetVisibility (Variables|Default|GetSM_C4_Explosive) false false)
      (if (Math|Boolean|ORBoolean (Variables|Default|GetbExploded) (Variables|Default|GetbExplodeOnDeath) (> Damage 0))
        (Audio|PlaySoundatLocation (Variables|Default|GetPreExplosion_Sound) (Transformation|GetWorldLocation (Variables|Default|GetSM_C4_Explosive)))
        (Audio|PlaySoundatLocation _explosion_sound _returnvalue) ; Explosion sound
        (Fortnite|ShowSoundIndicatorLocal 0 _returnvalue (Variables|/Script/Engine.SoundBase|GetMaxDistance _explosion_sound) "Gunshot")
        (if (Variables|Default|GetInWater)
          (Audio|PlaySoundatLocation (Variables|Default|GetWaterExplosionSound) _returnvalue "0, 0, 0" 1.2) ; Water Debris Sound
          (else
            _))
        (Fortnite|ShowSoundIndicatorLocalforTeam :Location _returnvalue :MaxAudibleDistance (Variables|/Script/Engine.SoundBase|GetMaxDistance _explosion_sound) :IndicatorType "Gunshot" :Team (Team|GetActorTeam self) :Affiliations (Utilities|Array|MakeArray "Hostile" "Neutral") :Tint "(R=0.000000,G=0.000000,B=0.000000,A=0.000000)" :AttenuationMaxRangeMultiplier 1.0 :AttenuationIntensityMultiplier 1.0)
        (bind _returnvalue_2 (Effects|Components|ParticleSystem|SpawnFXAtLocation (select (Variables|Default|GetInWater) (Variables|Default|GetNS_ExplodeWater) (Variables|Default|GetNS_Explode)) _returnvalue (Math|Rotator|MakeRotfromX (Variables|Default|GetHitNormal))))
        (Effects|Components|ParticleSystem|SetFloatParameter _returnvalue_2 "GlobalScale" 1.75)
        (Rendering|SetRenderCustomDepth _returnvalue_2 true)
        (Utilities|FNSpawnRadialForce (Math|Transform|MakeTransform _returnvalue) 1.0 0.66 7.5e+02 2.0
          (:Then
            (Fortnite|PlayLocalForceFeedbackatLocationMulti (Variables|Default|GetExplosionForceFeedbackNear) (Variables|Default|GetExplosionForceFeedbackFar) (Transformation|GetActorLocation) (Variables|Default|GetExplosionRadius) (* (Variables|Default|GetExplosionRadius) 3) "Explosion")
            (Camera|PlayWorldCameraShake "/Game/Abilities/Player/Commando/Actives/FragGrenade/B_GrenadeCameraShake.B_GrenadeCameraShake_C" (Transformation|GetActorLocation) 200 800)))))))

(event EventEnteredAirVent
  (params (LaunchVelocity :type Float%20%28double-precision%29) (BaseForceScale :type Float%20%28double-precision%29) (Direction :type Vector :default "0, 0, 0"))
  ;[1 0.1851 0.9664] Based on logic from B_HVAC_Prop
  (CallFunction|InvertProjectileGravity BaseForceScale))

(event Interaction|EventBlueprintOnInteract
  (params (InteractingPawn :type Fort%20Pawn%20Object%20Reference) (InteractionBeingAttempted :type EInteractionBeingAttempted%20Enum :default "FirstInteraction"))
  ;[0.09279 1 0.4823] On Interact
  (Variables|Interaction|SetbAllowInteract false)
  (Inventory|GiveItemtoInventoryOwner (Fortnite|GetFortPlayerControllerfromActor InteractingPawn) "/Figment_C4_Athena/Gameplay/Figment_C4_Athena.Figment_C4_Athena")
  (Variables|Default|SetAllowExplode? false)
  (Variables|Default|SetbExplodeOnDeath false)
  (Variables|Default|SetIsDead? true)
  (Actor|DestroyActor))

(event EventExitedAirVent
  (params (LaunchVelocity :type Float%20%28double-precision%29) (BaseForceScale :type Float%20%28double-precision%29) (Direction :type Vector :default "0, 0, 0"))
  ;[1 0.1851 0.9664] Based on logic from B_HVAC_Prop
  (CallFunction|InvertProjectileGravity BaseForceScale))

(fn BlueprintCanInteract
  (params (InteractingPawn :type Fort%20Pawn%20Object%20Reference) (InteractionBeingAttempted :type EInteractionBeingAttempted%20Enum :default "FirstInteraction") (InteractionType :type TInteractionType%20Enum :default "IT_NoInteraction"))
  (return true))

(fn BlueprintGetInteractionString
  (params (InteractingPawn :type Fort%20Pawn%20Object%20Reference) (InteractionBeingAttempted :type EInteractionBeingAttempted%20Enum :default "FirstInteraction"))
  (return (Variables|Default|GetInteractText)))

(fn GetAffiliationsToShowFor
  (return (Utilities|Array|MakeArray "Friendly" "Neutral" "Hostile")))

(fn GetTargetingSourceTransform
  (params (Source :type EFortAbilityTargetingSource%20Enum :default "Camera"))
  (return (Math|Transform|MakeTransform (+ (.location (|CallParentFunction Source)) (Math|Vector|Normalize (Variables|Default|GetHitNormal))) (.rotation (|CallParentFunction Source)) (.scale (|CallParentFunction Source)))))

(fn HandleHitBuildingActor
  (params (Object :type Object%20Reference) (Component :type Object%20Reference))
  (Utilities|Casting|CastToBuildingActor Object)
  (bind _asscene_component (Utilities|Casting|CastToSceneComponent Component))
  (Variables|Default|SetHandleHitPassed true)
  (Transformation|AttachActorToComponent _asscene_component "None" "KeepWorld" "KeepWorld" "KeepWorld" false))

(fn HandleHitBuildingWall
  (params (object :type Object%20Reference) (Comp :type Primitive%20Component%20Object%20Reference))
  (bind _asbuilding_wall (Utilities|Casting|CastToBuildingWall object))
  (LocalVariables|SetLocalWall _asbuilding_wall)
  (Variables|Default|SetHandleHitPassed true)
  (bind _asstatic_mesh_component (Utilities|Casting|CastToStaticMeshComponent (Variables|GetComp)))
  (bind _returnvalue (BuildingWall|IsDoorComponent (LocalVariables|GetLocalWall) _asstatic_mesh_component))
  (if _returnvalue
    (Transformation|AttachActorToComponent (Variables|GetComp) "None" "KeepWorld" "KeepWorld" "KeepWorld" false)))

(fn HandleHitSupplyDrop
  (params (Object :type Object%20Reference) (Component :type Primitive%20Component%20Object%20Reference))
  (Utilities|FlowControl|Sequence
    (:then_0
      (Utilities|Casting|CastToFortAthenaSupplyDrop Object)
      (Transformation|AttachActorToComponent Component "None" "KeepWorld" "KeepWorld" "KeepWorld" false)
      (Variables|Default|SetHandleHitPassed true))))

(fn HandleHitVehicleOrCube
  (params (Object :type Object%20Reference))
  ;[0.2395 1 0.8293] Handle hitting vehicle
  (bind _asfort_vehicle_interface (Utilities|Casting|CastToFortVehicleInterface Object))
  ;;[0.5525 1 0.3047] Adjust collision settings so we don't prevent pawns from riding in the vehicle
  (Collision|SetCollisionEnabled (Variables|Default|GetRoot_Sphere_Collision) "NoCollision")
  (Collision|SetCollisionResponsetoChannel (Variables|Default|GetSM_C4_Explosive) "ECC_Pawn" "ECR_Ignore")
  ;[0.3021 0.09068 0.2563 0.5] FORT-983093; to re-enabling collision once vehicle is destoyed
  (bind _returnvalue (Vehicle|GetVehicleActor _asfort_vehicle_interface :!class /Script/FortniteGame.FortVehicleInterface))
  (Game|BindEventtoOnDestroyed _returnvalue (EventDispatchers|CreateEvent :!event VehicleActorDestroyed)))

(fn HandleWasAttachedToWall
  (params (Object :type Object%20Reference))
  (bind (_outcomponents _returnvalue) (Collision|SphereOverlapComponents (Transformation|GetActorLocation) 16.0 (Utilities|Array|MakeArray "ObjectTypeQuery1") "/Script/Engine.StaticMeshComponent" (Utilities|Array|MakeArray self)))
  (if _returnvalue
    (CallFunction|HandleHitBuildingWall (Variables|GetObject) (Utilities|Array|Get(acopy) _outcomponents))))

(fn InvertProjectileGravity
  (params (DynamicForceScale :type Float%20%28double-precision%29))
  (Variables|/Script/Engine.ProjectileMovementComponent|SetProjectileGravityScale (Math|Float|float*float (Variables|/Script/Engine.ProjectileMovementComponent|GetProjectileGravityScale (Variables|Default|GetProjectileMovement1)) -1.0 DynamicForceScale) (Variables|Default|GetProjectileMovement1)))

(fn OnRep_bResumeSimulation
  (if (Variables|Default|GetbResumeSimulation)
    (Components|Movement|SetUpdatedComponent (Variables|Default|GetProjectileMovement1) (Variables|Default|GetRoot_Sphere_Collision))))

(fn OnRep_bTriggeredExplosion)

(fn OnRep_IsDead?
  (bind _isdead (Variables|Default|GetIsDead?))
  (bind _returnvalue (not _isdead))
  (Utilities|FlowControl|Sequence
    (:then_0
      (bind _ns_c4_flare (Variables|Default|GetNS_C4_Flare)
        (:then
          (Components|Activation|SetActive _ns_c4_flare _returnvalue false))
        (:else)))
    (:then_1
      (bind _fximpact (Variables|Default|GetFXImpact)
        (:then
          (Components|Activation|SetActive _fximpact _returnvalue false))
        (:else)))))

(fn OnRep_LandedTime
  (bind _returnvalue (Game|GetGameState))
  (bind _fortwaterinteraction (Variables|Default|GetFortWaterInteraction))
  (bind _returnvalue_2 (Water|IsinWater _fortwaterinteraction))
  (bind _returnvalue_3 (Transformation|GetActorLocation))
  (bind _sound_hitwater (Variables|Default|GetSound_HitWater))
  (bind _land_sound (Variables|Default|GetLand_Sound))
  (bind _returnvalue_4 (select _returnvalue_2 _sound_hitwater _land_sound))
  (Utilities|IsValid _returnvalue
    (:Is%20Valid
      (if (<= (- (GameState|GetServerWorldTimeSeconds _returnvalue) (Variables|Default|GetLandedTime)) 2.0)
        (if (not (Variables|Default|GetbPlayedLandingSound))
          (Audio|PlaySoundatLocation _returnvalue_4 _returnvalue_3)
          (Variables|Default|SetbPlayedLandingSound true)
          (SoundIndicator|ShowSoundIndicator (Variables|Default|GetSoundIndicator) _returnvalue_3 _returnvalue_4 1.0 0))))))

(fn ShouldShowSoundIndicator
  (params (PlayerController :type Fort%20Player%20Controller%20Object%20Reference))
  (return true))

(fn TryAttachToActor
  (params (Actor :type Actor%20Object%20Reference))
  (bind (_aspawn _bsuccess) (Utilities|Casting|CastToPawn Actor))
  (bind (_asfort_vehicle_interface _bsuccess_2) (Utilities|Casting|CastToFortVehicleInterface Actor))
  (bind (_asb_bga_athena_c4 _bsuccess_3) (Utilities|Casting|CastToB_BGA_Athena_C4 Actor))
  ;[0.3333 1 0.9901] Handle attaching to hit actor if not a pawn
  (Utilities|IsValid Actor
    (:Is%20Valid
      (Networking|SwitchHasAuthority
        (:Authority
          (if (and (or (not _bsuccess) _bsuccess_2) (not _bsuccess_3))
            (Variables|Default|SetHandleHitPassed true)
            (Gameplay|AuthorityAttachtoActor (Variables|Default|GetFortAttachableProjectile) (Variables|Default|GetHitResult) Actor 0.0 0.0 0.0)))))))

(fn ConstructionScript
  ;[1 1 1] Hot fix bool for if we want to explode on death. It is implied that if you want to explode, we destroy the C4 so that is also checked here
  (bind _outxy (DataTable|EvaluateCurveTableRow "/Game/Athena/Balance/DataTables/AthenaGameData.AthenaGameData" "Default.C4.ExplodeOnPlayerDeath")
    (:RowFound
      (bind _output_get (Variables|Default|SetbExplodeOnDeath (Math|Conversions|ToBoolean(Integer) (Math|Float|Truncate _outxy))))
      (if _output_get
        (Variables|Default|SetbDestroyOnDeath true)
        (else
          _)))
    (:RowNotFound
      _))
  ;[1 1 1] Hot fix bool for if we want to destroy on death
  (bind _outxy_2 (DataTable|EvaluateCurveTableRow "/Game/Athena/Balance/DataTables/AthenaGameData.AthenaGameData" "Default.C4.DestroyOnPlayerDeath")
    (:RowFound
      (Variables|Default|SetbDestroyOnDeath (Math|Conversions|ToBoolean(Integer) (Math|Float|Truncate _outxy_2))))
    (:RowNotFound
      _))
  (CallFunction|SetProjectile (Variables|/Script/Engine.ProjectileMovementComponent|GetInitialSpeed (Variables|Default|GetProjectileMovement1)) (Variables|/Script/Engine.ProjectileMovementComponent|GetProjectileGravityScale (Variables|Default|GetProjectileMovement1)) (Variables|Default|GetActivation_Delay))
  ;[1 1 1] PostThrowEndDelay variable. The delay that happens after you click the detonate button.
  (bind _outxy_3 (DataTable|EvaluateCurveTableRow "/Game/Athena/Balance/DataTables/AthenaGameData.AthenaGameData" "Default.C4.PostThrowEndDelay")
    (:RowFound
      (Development|PrintString (Utilities|String|ToString(Float) _outxy_3))
      (Variables|Default|SetPostThrowEndDelay _outxy_3))))
